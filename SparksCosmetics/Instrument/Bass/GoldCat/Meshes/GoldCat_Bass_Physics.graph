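# Physics asset for the GoldCat bass instrument mesh in SparksCosmetics, simulating a fish part linked to the handle by rings and connectors.
# Capsules cover the handle, fish and rings, with small spheres on the two connectors.
# Limited-swing constraints form a closed loop from handle through upper ring, connector, fish, bottom connector and bottom ring back to handle.
# Twist is locked at ring-to-connector joints, and the bottom pair is constrained in both directions.
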PhysicsAsset /SparksCosmetics/Instrument/Bass/GoldCat/Meshes/GoldCat_Bass_Physics
bodies {
 Bottom_connect [{shapeType=Sphere, radius=1, shapeName=Bottom_connect_sphere}]
 Bottom_ring [{shapeType=Capsule, length=1, radius=1, shapeName=Bottom_ring_capsule, center=<0.4602 0 0.0649>, rotation=<rot 81.9783 180 180>}]
 Upper_connect [{shapeType=Sphere, radius=1, shapeName=Upper_connect_sphere}]
 Upper_fish [{shapeType=Capsule, length=12.0808, radius=2, shapeName=Upper_fish_capsule, center=<8.4 0 -3>, rotation=<rot 76 0 0>}]
 Upper_ring [{shapeType=Capsule, length=1, radius=1, shapeName=Upper_ring_capsule, center=<0.4602 0 0.0649>, rotation=<rot 81.9783 180 180>}]
 handle [{shapeType=Capsule, length=10, radius=3}]
}
constraints [
 {bone1Name=Upper_ring, bone2Name=handle, swing1LimitDegrees=35, swing2LimitDegrees=35, twistLimitDegrees=5}
 {bone1Name=Upper_connect, bone2Name=Upper_ring, swing1LimitDegrees=60, swing2LimitDegrees=45, twistLimitDegrees=45, twistMotion=Locked}
 {bone1Name=Bottom_ring, bone2Name=handle, swing1LimitDegrees=35, swing2LimitDegrees=20, twistLimitDegrees=5}
 {bone1Name=Upper_fish, bone2Name=Upper_connect, swing1LimitDegrees=45, swing2LimitDegrees=45, twistLimitDegrees=5}
 {bone1Name=Bottom_connect, bone2Name=Upper_fish, swing1LimitDegrees=45, swing2LimitDegrees=45, twistLimitDegrees=5}
 {bone1Name=Bottom_ring, bone2Name=Bottom_connect, swing1LimitDegrees=60, swing2LimitDegrees=45, twistLimitDegrees=5, twistMotion=Locked}
 {bone1Name=Bottom_connect, bone2Name=Bottom_ring, swing1LimitDegrees=60, swing2LimitDegrees=45, twistLimitDegrees=5, twistMotion=Locked}
]
